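# Translucent unlit mesh-particle material for a ring effect in the Soldier's Going Commando ability.
# A noise texture pans in irregular jumps, driven by floored, sine-perturbed particle time, with speed and U scale from dynamic parameters.
# Ceil thresholds on the noise produce a hard-edged mask shifted by Transition and a two-tone inner/outer color split tinted by particle color.
# With depth testing disabled, a scene-versus-pixel depth comparison fades occluded parts by their distance behind geometry.
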Material /Game/CharClasses/Soldier/Traits/AdvancedTactics/Abilities/GoingCommando/Materials/M_GoingCommando_Ring_01
graph {
 properties {
  bDisableDepthTest=true
  bEnableResponsiveAA=true
  bUsedWithMeshParticles=true
  bUsedWithNiagaraMeshParticles=true
  blendMode=BLEND_Translucent
  pixelDepthOffsetMode=PDOM_Legacy
  shadingModel=MSM_Unlit
  usedShadingModels=Unlit
 }
 nodes [
  n1 = VectorParameter() {parameterName=Color Outer, defaultValue=<rgba 2 1.0108 0.09 0>}
  n2 = VectorParameter() {parameterName=Color Inner, defaultValue=<rgba 7 6.3786 4.165 0>}
  n3 = AppendVector(
   A=DynamicParameter() {defaultValue=<rgba 1.8 1 1 0>, paramNames [Panner Speed, U Scale, "2 (Not Used)", Transition]}.U Scale
   B=Constant() {r=1}
  )
  n4 = TextureCoordinate()
  n5 = Multiply(
   A=ParticleRelativeTime()
   B=DynamicParameter() {defaultValue=<rgba 1.8 1 1 0>, paramNames [Panner Speed, U Scale, "2 (Not Used)", Transition]}.Panner Speed
  )
  n6 = Sine(None=@n5) {period=0.77}
  n7 = Sine(None=@n5) {period=1.12}
  n8 = Add(A=Abs(None=Add(A=@n6, B=@n7)), B=@n5)
  n9 = TextureSampleParameter2D(UVs=Panner(Coordinate=Multiply(A=@n3, B=@n4), Time=Floor(None=@n8)) {bFractionalPart=true, speedX=-0.7}) {
   parameterName=Texture
   samplerType=SAMPLERTYPE_LinearColor
   texture=/Game/CharClasses/Soldier/Traits/AdvancedTactics/Abilities/GoingCommando/Textures/T_GoingCommando_RingNoise_01
  }
  n10 = Add(A=@n9.G, B=DynamicParameter() {defaultValue=<rgba 1.8 1 1 0>, paramNames [Panner Speed, U Scale, "2 (Not Used)", Transition]}.Transition)
  n11 = ScalarParameter() {defaultValue=-0.6, parameterName=Color Edge Fade Bias}
  n12 = Clamp(None=Ceil(None=Add(A=@n10, B=@n11)))
  MaterialFunctionCall(Emissive=Multiply(A=ParticleColor().RGB, B=LinearInterpolate(A=@n1.RGB, B=@n2.RGB, Alpha=@n12))) {materialFunction=&MF_FX_EmissiveAdjustments}
  n13 = Constant() {r=1}
  n14 = PixelDepth()
  n15 = SceneDepth()
  n16 = Clamp(None=Divide(A=Subtract(A=@n14.R, B=@n15.R), B=ScalarParameter() {defaultValue=150, parameterName=Depth Fade Behind Scale}))
  Multiply(A=Ceil(None=@n10), B=If(A=SceneDepth().R, B=PixelDepth().R, "A > B"=@n13, "A == B"=@n13, "A < B"=OneMinus(None=@n16)))
  Sine(None=Time())
  Time()
 ]
}
