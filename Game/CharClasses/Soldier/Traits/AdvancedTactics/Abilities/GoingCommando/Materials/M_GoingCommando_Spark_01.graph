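# Translucent unlit spark material for particle and Niagara sprites in the Soldier's Going Commando ability effects.
# Particle color RGB passes through the MF_FX_EmissiveAdjustments function as its emissive input.
# Particle alpha multiplies the RGB of the T_SparkDirection texture, so the texture shapes the spark while the emitter controls color and fade.
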Material /Game/CharClasses/Soldier/Traits/AdvancedTactics/Abilities/GoingCommando/Materials/M_GoingCommando_Spark_01
graph {
 properties {
  bUsedWithNiagaraSprites=true
  bUsedWithParticleSprites=true
  blendMode=BLEND_Translucent
  pixelDepthOffsetMode=PDOM_Legacy
  shadingModel=MSM_Unlit
  usedShadingModels=Unlit
 }
 nodes [
  n1 = ParticleColor()
  MaterialFunctionCall(Emissive=@n1.RGB) {materialFunction=&MF_FX_EmissiveAdjustments}
  Multiply(
   A=@n1.A
   B=TextureSample(UVs=TextureCoordinate()) {texture=/Game/CharClasses/Soldier/Traits/AdvancedTactics/Abilities/GoingCommando/Textures/T_SparkDirection}.RGB
  )
 ]
}
