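# Translucent unlit streak material for mesh particles in the Soldier's Going Commando ability effect.
# Two channels of one streak texture pan at different speeds, stepped by floored, sine-wobbled particle lifetime for a jumpy flicker, with vertex red as mask.
# The Transition dynamic parameter erodes this mask, blending outer to inner color for emissive and producing a hard Ceil-cut alpha shaped by vertex blue and optional particle alpha.
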Material /Game/CharClasses/Soldier/Traits/AdvancedTactics/Abilities/GoingCommando/Materials/M_GoingCommando_StreaksA_01
graph {
 properties {
  bEnableResponsiveAA=true
  bUsedWithMeshParticles=true
  bUsedWithNanite=true
  bUsedWithNiagaraMeshParticles=true
  blendMode=BLEND_Translucent
  pixelDepthOffsetMode=PDOM_Legacy
  shadingModel=MSM_Unlit
  usedShadingModels=Unlit
 }
 nodes [
  ConstantBiasScale(None=Sine(None=Time())) {bias=0, scale=2}
  n1 = ParticleColor()
  n2 = VectorParameter() {parameterName=Color Outer, defaultValue=<rgba 2 1.0108 0.09 0>}
  n3 = VectorParameter() {parameterName=Color Inner, defaultValue=<rgba 7 6.3786 4.165 0>}
  n4 = VertexColor()
  n5 = AppendVector(
   A=DynamicParameter() {defaultValue=<rgba 1.8 1 1 0>, paramNames [Panner Speed, U Scale, "2 (Not Used)", Transition]}.U Scale
   B=Constant() {r=1}
  )
  n6 = Multiply(
   A=ParticleRelativeTime()
   B=DynamicParameter() {defaultValue=<rgba 1.8 1 1 0>, paramNames [Panner Speed, U Scale, "2 (Not Used)", Transition]}.Panner Speed
  )
  n7 = Add(A=Sine(None=@n6) {period=0.77}, B=Sine(None=@n6) {period=1.12})
  n8 = Floor(None=Add(A=Abs(None=@n7), B=@n6))
  n9 = Panner(Coordinate=Multiply(A=@n5, B=TextureCoordinate()), Time=@n8) {bFractionalPart=true, speedX=-0.7}
  n10 = Panner(Coordinate=Multiply(A=@n5, B=TextureCoordinate() {uTiling=0.6}), Time=@n8) {bFractionalPart=true, speedX=-7.89}
  n11 = Multiply(
   A=@n4.R
   B=Add(
    A=TextureSample(UVs=@n9) {samplerType=SAMPLERTYPE_LinearColor, texture=/Game/Effects/Fort_Effects/Gameplay/SmokeDustExplosions/Textures/T_Explosion_Streak_01}.R
    B=TextureSample(UVs=@n10) {samplerType=SAMPLERTYPE_LinearColor, texture=&T_Explosion_Streak_01}.G
   )
  )
  n12 = DynamicParameter() {defaultValue=<rgba 1.8 1 1 0>, paramNames [Panner Speed, U Scale, "2 (Not Used)", Transition]}
  n13 = Saturate(None=Subtract(A=Subtract(A=@n11, B=@n12.Transition) {constB=0.5}, B=OneMinus(None=@n4.B)))
  MaterialFunctionCall(Emissive=Multiply(A=@n1.RGB, B=LinearInterpolate(A=@n2.RGB, B=@n3.RGB, Alpha=@n13))) {materialFunction=&MF_FX_EmissiveAdjustments}
  n14 = Subtract(A=Add(A=@n11, B=ScalarParameter() {defaultValue=0.8, parameterName=Alpha Edge Fade Bias}), B=@n12.Transition) {constB=0.5}
  n15 = Multiply(A=Ceil(None=Clamp(None=@n14)), B=VertexColor().B)
  StaticSwitchParameter(True=Multiply(A=@n1.A, B=@n15), False=@n15) {parameterName=UseParticleAlpha}
  Time()
 ]
}
